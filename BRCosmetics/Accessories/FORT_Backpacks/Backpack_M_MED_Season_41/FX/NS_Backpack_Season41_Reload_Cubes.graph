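# Niagara effect for the Season 41 Reload backpack cosmetic, drawing glowing cubes linked by lightning on the backpack's skeletal mesh.
# Cubes pins particles to bones by partition index, with per-index scale arrays, glow sprites and lights.
# Lightning stretches mesh bolts between randomly selected sampled points, while LightningChains fires timed bursts along one of four preset bone sequences, adding glows.
# A SmallCubes emitter of drifting cubes is present but disabled.
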NiagaraSystem /BRCosmetics/Accessories/FORT_Backpacks/Backpack_M_MED_Season_41/FX/NS_Backpack_Season41_Reload_Cubes
dependencies {
 usedDataInterfaces [
  &NiagaraDataInterfaceArrayFloat
  &NiagaraDataInterfaceArrayFloat3
  &NiagaraDataInterfaceArrayInt32
  &NiagaraDataInterfaceCurve
  &NiagaraDataInterfaceArrayFloat2
  &NiagaraDataInterfaceSkeletalMesh
  &NiagaraDataInterfaceColorCurve
 ]
 usedDynamicInputs []
 usedModules [
  &self:SystemSpawnScript.SystemScriptSource.SystemScriptGraph.NiagaraNodeAssignment_0.SetVariables_#1
  &self:SystemSpawnScript.SystemScriptSource.SystemScriptGraph.NiagaraNodeAssignment_1.SetVariables_#2
  &SystemState
 ]
}
emitters {
 Cubes {
  emitterUpdateScript [
   &EmitterState(Life Cycle Mode=::NewEnumerator1, Loop Delay=$System.Delay, Loop Duration=500, UseLoopDelay=false)
   &SpawnBurst_Instantaneous(Spawn Count=12, Use Spawn Probability=false, UseLoopCountLimit=true)
   &SpawnBurst_Instantaneous(Spawn Group=1, Use Spawn Probability=false, UseLoopCountLimit=true) {module=SpawnBurst_Instantaneous001}
   &SpawnBurst_Instantaneous(Spawn Time=0.3333, Use Spawn Probability=false, UseLoopCountLimit=false) {module=SpawnBurst_Instantaneous002}
   &SpawnBurst_Instantaneous(Spawn Time=0.4333, Use Spawn Probability=false, UseLoopCountLimit=false) {module=SpawnBurst_Instantaneous003}
  ]
  particleSpawnScript [
   &PartitionParticles(Number Of Partitions=12, Particles Per Partition=1, Partition Mode:ENiagara_PartitionMode_v2=::NewEnumerator1)
   &InitializeParticle(
    Color=<rgba 0.0696 0.3075 0.3229 1>
    Lifetime=2
    Mesh Scale Mode=::NewEnumerator3
    Mesh Uniform Scale=0.35
    UsePositionOffset=false
    Write Material Random=true
   )
   /SetVariables(Particles.Scale Array:NiagaraFloat=&SelectFloatFromArray(
     Direct Array Index=$Particles.PartitionParticles.Partition
     Float Selection Array=DataInterface() {propertyValues {FloatData [0.21,0.26,0.34,0.12,0.27,0.34,0.71,0.83,0.58,0.5,0.6,0.45]}}
    )) {isSetParameters=true}
   /SetVariables(Particles.Scale=&VectorFromFloat(Value=$Particles.Scale Array)) {isSetParameters=true}
   /SetVariables(Particles.Scale=&Multiply_Vector(
     A=$Particles.Scale
     B=&SelectVectorFromArray(
      Direct Array Index=$Particles.PartitionParticles.Partition
      Vector Selection Array=DataInterface() {propertyValues {FloatData [<1 1 1>, <1 1 1>, <1 1 1>, <1 1 1>, <1 1 1>, <1 1 1>, <0.9 1 1>, <1 0.8 1>, <1.1 0.9 1>, <1 0.7 1>, <0.8 1 1>, <1 1 1>]}}
     )
    )) {isSetParameters=true}
   /SetVariables(Particles.VisibilityTag=/Niagara/DynamicInputs/Helpers/MaskInt(
     Mask Source:NiagaraInt32=1
     Mask Parameter:NiagaraInt32=$Emitter.SpawnGroup
     Mask:NiagaraInt32=1
     Masked Int:NiagaraInt32=$Emitter.SpawnGroup
     Passthrough Int:NiagaraInt32=&SelectIntFromArray(
      Direct Array Index=$Particles.PartitionParticles.Partition
      Int Selection Array=DataInterface() {propertyValues {IntData [4,0,0,3,4,0,2,2,2,2,2,2,2]}}
     )
    )) {isSetParameters=true}
   /SetVariables(Particles.PartitionParticles.Partition:NiagaraInt32=&MaskInt(
     Mask Source:NiagaraInt32=1
     Mask Parameter:NiagaraInt32=$Particles.VisibilityTag
     Mask:NiagaraInt32=1
     Masked Int:NiagaraInt32=&RandomRangeInt(Minimum=7, Maximum=9)
     Passthrough Int:NiagaraInt32=$Particles.PartitionParticles.Partition
    )) {isSetParameters=true}
   &InitialMeshOrientation(
    Mesh Orientation Mode=::NewEnumerator7
    Rotation=&SelectVectorFromArray(
     Direct Array Index=$Particles.PartitionParticles.Partition
     Vector Selection Array=DataInterface() {propertyValues {FloatData [<0 0 0>, <0 0 0>, <0 0 0>, <0 0 0>, <0 0 0.015>, <0 0 0>, <0 0 0>, <0 0 0>, <0 0 0>, <0 0 0>, <0 0 0>, <0 0 0>]}}
    )
    Use Rotation Vector=true
   )
  ]
  particleUpdateScript [
   &ParticleState(Kill Particles When Lifetime Has Elapsed=false, Let Infinitely Lived Particles Die When Emitter Deactivates=true)
   /SetVariables(Particles.OutlineColor:LinearColor=&MultiplyLinearColor(Linear Color=<rgba 1 1 1 1>, Scale Factor=&MakeVector4(X=2, Y=2, Z=2, W=0.8))) {isSetParameters=true}
   &SkeletalMeshLocation(
    Bone / Socket Index=$Particles.PartitionParticles.Partition
    Bone Sampling Mode=::NewEnumerator4
    MaskSpawnGroup=false
    Skeletal Mesh=$System.Skeletal Mesh Reader
   )
   /SetVariables(
    Particles.Glow_Base_CameraOffset=-5
    Particles.Glow_Base_Color:LinearColor=&Multiply_LinearColorByFloat(
     Linear Color=<rgba 0.395 2.6551 3 0.2>
     Float=&MaskFloat(
      Mask Source=1
      Mask Parameter=$Particles.PartitionParticles.Partition
      Mask=0
      Masked Float=&FloatFromCurve(
       FloatCurve=DataInterface() {propertyValues {Curve {keys [<at 0 1>, {arriveTangent=1.2666, leaveTangent=1.2666, time=1.5, value=1}, <at 1.5791 3>, <at 1.7 1>]}}}
       CurveIndex=$Emitter.Age
      )
      Passthrough Float=1
     )
    )
    Particles.Glow_Base_Size:Vector2f=&Multiply_Vector2D_ByFloat(
     Float=&SelectFloatFromArray(
      Direct Array Index=$Particles.PartitionParticles.Partition
      Float Selection Array=DataInterface() {propertyValues {FloatData [0.5,0.5,0.65,0.45,0.6,0.7,1,1.5,1.3,0.7,1.7,0.84]}}
     )
     Vector2D=<22 22>
    )
   ) {isSetParameters=true}
   &Light_Attributes(
    Color=1
    Color Value To Scale:LinearColor=/Niagara/DynamicInputs/LinearColor/MakeLinearColorFromVector4(Vector:Vector4f=<1 1 1 1>)
    Falloff Exponent=1
    Light Exponent=2
    Light Position Offset:Vector3f=<0 -11 0>
    Light Radius=20
    Position=1
    Position Offset Coordinate Space:ENiagaraCoordinateSpace=::Local
    Radius=1
    Scale Alpha=2
    Scale RGB:Vector3f=&MakeVectorFromLinearColorRGB(Color=<rgba 0.0557 0.676 1 1>)
    ScaleA=true
    ScaleRGB=true
    UsePositionOffset=true
   )
   /SetVariables(Particles.Scale=&Multiply_Vector(
     A=&Multiply_Vector(
      A=&Multiply_Vector(A=$Particles.SkeletalMeshLocation.SampledScale, B=&VectorFromFloat(Value=1.15))
      B=$Particles.Initial.Scale
     )
     B=$Engine.Owner.Scale
    )) {isSetParameters=true}
   &DynamicMaterialParameters(
    Index 0 Param 1=&RandomRangeFloat(Minimum=0, Maximum=1)
    Index 0 Param 2=&RandomRangeFloat(Minimum=0, Maximum=1)
    Index 0 Param 3=&RandomRangeFloat(Minimum=0, Maximum=1)
    Index 0 Param 4=&RandomRangeFloat(Minimum=0, Maximum=1)
    Module.Index0Float0Write=true
    Module.Index0Float1Write=true
    Module.Index0Float2Write=true
    Module.Index0Float3Write=true
   )
  ]
  rendererClasses [
   &NiagaraMeshRendererProperties
   &NiagaraSpriteRendererProperties
   &NiagaraLightRendererProperties
  ]
  renderers [
   {rendererClass=&NiagaraMeshRendererProperties, rendererIndex=0}
   {rendererClass=&NiagaraMeshRendererProperties, rendererIndex=1}
   {rendererClass=&NiagaraMeshRendererProperties, rendererIndex=2}
   {rendererClass=&NiagaraMeshRendererProperties, rendererIndex=3}
   {rendererClass=&NiagaraSpriteRendererProperties, rendererIndex=4}
   {rendererClass=&NiagaraLightRendererProperties, rendererIndex=5}
  ]
 }
 Lightning {
  emitterUpdateScript [
   &EmitterState()
   &SpawnRate(SpawnRate=6, Use Spawn Probability=false)
  ]
  particleSpawnScript [
   &PartitionParticles(
    A:NiagaraFloat=&RandomRangeFloat(Minimum=0, Maximum=1)
    Partition Mode:ENiagara_PartitionMode_v2=::NewEnumerator4
   )
   &InitializeParticle(
    Color Maximum=<rgba 0.1312 0.6268 1.25 1>
    Color Minimum=<rgba 0.0862 0.2474 1 1>
    Color Mode=::NewEnumerator2
    Lifetime=&RandomRangeFloat(Minimum=0.4, Maximum=0.7)
    Mesh Scale Mode=::NewEnumerator3
    UsePositionOffset=false
    Write Material Random=true
   )
   /SetVariables(Particles.SetSelector:NiagaraInt32=&RandomRangeInt(Minimum=0, Maximum=8)) {isSetParameters=true, module=SetVariables_#3}
   &DynamicMaterialParameters(
    Index 0 Param 1=&FloatFromCurve(
     FloatCurve=DataInterface() {propertyValues {Curve {keys [<at 0 1>,<at 1 1>]}}}
     Scale Curve=&SelectFloatFromArray(
      Direct Array Index=$Particles.Initial.SetSelector
      Float Selection Array=DataInterface() {propertyValues {FloatData [3,3,3,1.4,1.4,3,3,3,3]}}
     )
    )
    Index 0 Param 2=&UniformRangedFloat(Minimum=0.5, Maximum=0.3)
    Index 0 Param 3=&UniformRangedFloat(Minimum=0, Maximum=0)
    Index 0 Param 4=&UniformRangedFloat(Minimum=0, Maximum=0)
    Index 1 Param 1=&UniformRangedFloat(Minimum=-1.5, Maximum=-0.9)
    Module.Index0Float0Write=true
    Module.Index0Float1Write=true
    Module.Index0Float2Write=true
    Module.Index0Float3Write=true
    Module.Index1Float0Write=true
    Module.Index1Float1Write=true
    Module.Index1Float2Write=true
    Module.Index1Float3Write=true
    Write Parameter Index 1=::NewEnumerator1
   )
   /SetVariables(
    Particles.RareSetA:NiagaraInt32=&RandomRangeInt(Minimum=6, Maximum=10)
    Particles.SetA:NiagaraInt32=&SelectIntFromArray(Direct Array Index=$Particles.SetSelector, Int Selection Array=DataInterface() {propertyValues {IntData [0,0,1,3,3,2,1,4,4]}})
    Particles.SetB:NiagaraInt32=&SelectIntFromArray(Direct Array Index=$Particles.SetSelector, Int Selection Array=DataInterface() {propertyValues {IntData [1,2,2,5,6,6,6,1,6]}})
   ) {isSetParameters=true, module=SetVariables_#4}
  ]
  particleUpdateScript [
   &ParticleState()
   &SkeletalMeshLocation(
    Bone / Socket Index=&MaskInt(
     Mask Source:NiagaraInt32=1
     Mask Parameter:NiagaraInt32=$Particles.PartitionParticles.Partition
     Mask:NiagaraInt32=0
     Masked Int:NiagaraInt32=$Particles.SetA
     Passthrough Int:NiagaraInt32=$Particles.RareSetA
    )
    Bone Sampling Mode=::NewEnumerator4
    MaskSpawnGroup=false
    Skeletal Mesh=$System.Skeletal Mesh Reader
   )
   &SkeletalMeshLocation(
    Bone / Socket Index=$Particles.SetB
    Bone Sampling Mode=::NewEnumerator4
    MaskSpawnGroup=false
    Position Sampling=::NewEnumerator2
    Skeletal Mesh=$System.Skeletal Mesh Reader
   ) {module=SkeletalMeshLocation001}
   &SkeletalMeshLocation(
    MaskSpawnGroup=false
    Mesh Sampling Type=::NewEnumerator5
    OverrideRandomSeed=false
    Position Sampling=::NewEnumerator2
    Skeletal Mesh=$System.Skeletal Mesh Reader
   ) {module=SkeletalMeshLocation002}
   &self:ReadDeath(ChainTo:NiagaraPosition=&MaskVector(
     Mask Source=0
     Mask Parameter=?
     Mask=$Particles.PartitionParticles.Partition
     Masked Vector=Linked() {linkedVariable {name=Particles.SkeletalMeshLocation001.SampledPosition, type=NiagaraPosition}}
     Passthrough Vector=Linked() {linkedVariable {name=Particles.SkeletalMeshLocation002.SampledPosition, type=NiagaraPosition}}
    )) {module=ReadDeath}
   &ScaleMeshSize(Scale Factor=&MakeVector(
     X=&Divide_Float(
      A=&Lerp_Float(
       A=0
       B=$Particles.LightningLength
       Alpha=&FloatFromCurve(FloatCurve=DataInterface() {propertyValues {Curve {defaultValue=1, keys [{tangentMode=RCTM_User, time=0, value=0}, {tangentMode=RCTM_User, time=0.2, value=1}]}}})
      )
      B=100
     )
     Y=1
     Z=1
    ))
   &ScaleColor(
    Scale Alpha=&FloatFromCurve(FloatCurve=DataInterface() {propertyValues {Curve {
        defaultValue=1
        keys [
         {interpMode=RCIM_Cubic, tangentMode=RCTM_User, time=0, value=0}
         {interpMode=RCIM_Cubic, tangentMode=RCTM_User, time=0.1482, value=1}
         {tangentMode=RCTM_User, time=1, value=0}
        ]
       }}})
    Scale RGB=<5 5 5>
    ScaleA=true
    ScaleRGB=true
   )
  ]
  rendererClasses [&NiagaraMeshRendererProperties]
  renderers [{rendererClass=&NiagaraMeshRendererProperties, rendererIndex=0}]
 }
 LightningChains {
  emitterUpdateScript [
   &EmitterState(
    Inactive Response=::NewEnumerator1
    Life Cycle Mode=::NewEnumerator1
    Loop Delay=&RandomRangeFloat(Minimum=0.1, Maximum=0.3, Recalculate Random Each Loop=false)
    Loop Duration=&RandomRangeFloat(Minimum=0.6, Maximum=1, Evaluation Type=::NewEnumerator1)
    Recalculate Duration Each Loop=true
    UseLoopDelay=false
   )
   /SetVariables(
    Emitter.Lifetime:NiagaraFloat=&RandomRangeFloat(Minimum=1, Maximum=1)
    Emitter.SetSelector:NiagaraInt32=&RandomRangeInt(Minimum=0, Maximum=3)
   ) {isSetParameters=true}
   /SetVariables(Emitter.SetSelector=1) {enabled=false, isSetParameters=true}
   &SpawnBurst_Instantaneous(Use Spawn Probability=false, UseLoopCountLimit=false)
   &SpawnBurst_Instantaneous(Spawn Group=4, Spawn Time=0.16, Use Spawn Probability=false, UseLoopCountLimit=false) {module=SpawnBurst_Instantaneous004}
   &SpawnBurst_Instantaneous(Spawn Group=1, Spawn Time=0.15, Use Spawn Probability=false, UseLoopCountLimit=false) {module=SpawnBurst_Instantaneous001}
   &SpawnBurst_Instantaneous(Spawn Group=5, Spawn Time=0.28, Use Spawn Probability=false, UseLoopCountLimit=false) {module=SpawnBurst_Instantaneous005}
   &SpawnBurst_Instantaneous(Spawn Group=2, Spawn Time=0.3, Use Spawn Probability=false, UseLoopCountLimit=false) {module=SpawnBurst_Instantaneous002}
   &SpawnBurst_Instantaneous(Spawn Group=6, Spawn Time=0.45, Use Spawn Probability=false, UseLoopCountLimit=false) {module=SpawnBurst_Instantaneous006}
   &SpawnBurst_Instantaneous(Spawn Group=3, Spawn Time=0.45, Use Spawn Probability=false, UseLoopCountLimit=false) {module=SpawnBurst_Instantaneous003}
   &SpawnBurst_Instantaneous(Spawn Group=7, Spawn Time=0.55, Use Spawn Probability=false, UseLoopCountLimit=false) {module=SpawnBurst_Instantaneous007}
  ]
  particleSpawnScript [
   &PartitionParticles(
    A:NiagaraFloat=&RandomRangeFloat(Minimum=0, Maximum=1)
    B=0.5
    Partition Mode:ENiagara_PartitionMode_v2=::NewEnumerator4
   )
   /SetVariables(Particles.SetPage:NiagaraInt32=$Emitter.SpawnGroup) {isSetParameters=true}
   /SetVariables(Particles.VisibilityTag=&MakeCustomIntFromBool(
     Bool=&SetBoolByIntComparison(A=$Emitter.SpawnGroup, B=4, Comparison Type=::NewEnumerator2)
     True INT=1
     False INT=0
    )) {isSetParameters=true}
   /SetVariables(Particles.SetPage:NiagaraInt32=&MaskInt(
     Mask Source:NiagaraInt32=1
     Mask Parameter:NiagaraInt32=&MakeCustomIntFromBool(
      Bool=&SetBoolByIntComparison(A=$Particles.SetPage, B=4, Comparison Type=::NewEnumerator2)
      True INT=1
      False INT=0
     )
     Mask:NiagaraInt32=0
     Masked Int:NiagaraInt32=$Particles.SetPage
     Passthrough Int:NiagaraInt32=/Niagara/DynamicInputs/Subtract/Subtract_Integer(A:NiagaraInt32=$Particles.SetPage, B:NiagaraInt32=4)
    )) {isSetParameters=true}
   &InitializeParticle(
    Color Maximum=<rgba 0.1312 0.6268 1.25 1>
    Color Minimum=<rgba 0.0862 0.2474 1 1>
    Color Mode=::NewEnumerator2
    Lifetime=$Emitter.Lifetime
    Mesh Scale Mode=::NewEnumerator3
    UsePositionOffset=false
    Write Material Random=true
   )
   /SetVariables(Particles.SetSelector:NiagaraInt32=$Emitter.SetSelector) {isSetParameters=true, module=SetVariables_#3}
   &DynamicMaterialParameters(
    Index 0 Param 1=&FloatFromCurve(
     FloatCurve=DataInterface() {propertyValues {Curve {keys [<at 0 1>,<at 1 1>]}}}
     Scale Curve=&SelectFloatFromArray(
      Direct Array Index=$Particles.Initial.SetSelector
      Float Selection Array=DataInterface() {propertyValues {FloatData [3,3,3,1.4,1.4,3,3,3,3]}}
     )
    )
    Index 0 Param 2=&UniformRangedFloat(Minimum=0.5, Maximum=0.3)
    Index 0 Param 3=&UniformRangedFloat(Minimum=0, Maximum=0)
    Index 0 Param 4=&UniformRangedFloat(Minimum=0, Maximum=0)
    Index 1 Param 1=&UniformRangedFloat(Minimum=-1.5, Maximum=-0.9)
    Module.Index0Float0Write=true
    Module.Index0Float1Write=true
    Module.Index0Float2Write=true
    Module.Index0Float3Write=true
    Module.Index1Float0Write=true
    Module.Index1Float1Write=true
    Module.Index1Float2Write=true
    Module.Index1Float3Write=true
    Write Parameter Index 1=::NewEnumerator1
   )
   /SetVariables(
    Particles.SetA:NiagaraInt32=&MergeHelper_IntSwitch(
     0=&MergeHelper_IntSwitch(0=5, 1=3, 2=1, 3=0, 4=?, 5=?, Index=$Particles.SetPage, Num=::NewEnumerator2)
     1=&MergeHelper_IntSwitch(0=11, 1=4, 2=2, 3=0, 4=?, 5=?, Index=$Particles.SetPage, Num=::NewEnumerator2)
     2=&MergeHelper_IntSwitch(0=5, 1=2, 2=3, 3=6, 4=?, 5=?, Index=$Particles.SetPage, Num=::NewEnumerator2)
     3=&MergeHelper_IntSwitch(0=2, 1=1, 2=0, 3=6, 4=?, 5=?, Index=$Particles.SetPage, Num=::NewEnumerator2)
     4=?
     5=?
     Index=$Emitter.SetSelector
     Num=::NewEnumerator2
    )
    Particles.SetB:NiagaraInt32=&MergeHelper_IntSwitch(
     0=&MergeHelper_IntSwitch(0=6, 1=5, 2=3, 3=1, 4=?, 5=?, Index=$Particles.SetPage, Num=::NewEnumerator2)
     1=&MergeHelper_IntSwitch(0=11, 1=6, 2=4, 3=2, 4=?, 5=?, Index=$Particles.SetPage, Num=::NewEnumerator2)
     2=&MergeHelper_IntSwitch(0=8, 1=5, 2=2, 3=6, 4=?, 5=?, Index=$Particles.SetPage, Num=::NewEnumerator2)
     3=&MergeHelper_IntSwitch(0=8, 1=2, 2=1, 3=6, 4=?, 5=?, Index=$Particles.SetPage, Num=::NewEnumerator2)
     4=?
     5=?
     Index=$Emitter.SetSelector
     Num=::NewEnumerator2
    )
   ) {isSetParameters=true, module=SetVariables_#4}
   /SetVariables(
    Particles.CubeGlowColor:LinearColor=<rgba 0 0.9044 1 1>
    Particles.GlowColor=&MultiplyLinearColor(
     Linear Color=<rgba 0.1188 0.5928 1 1>
     Scale Factor=&MakeVector4FromVectorAndFloat(
      "Vector (RGB)"=&VectorFromFloat(Value=&FN_FrontEndFloatModifier(Gameplay=1, Front End=1))
      "Float (Alpha)"=&FNTimeOfDayScalar(Night Scale=0.26, Morning Scale=0.3, Day Scale=0.5, Evening Scale=0.3)
     )
    )
    Particles.GlowSize=&SelectVector2DFromArray(
     Direct Array Index=$Particles.SetA
     Vector 2D Selection Array=DataInterface() {propertyValues {FloatData [<7.5 7.5>, <9 9>, <12 12>, <6 6>, <10 10>, <12 12>, <0 0>, <0 0>, <0 0>, <0 0>, <0 0>, <12 12>, <0 0>]}}
    )
   ) {isSetParameters=true}
   /SetVariables(Particles.CubeGlowScale:Vector3f=&VectorFromFloat(Value=&SelectFloatFromArray(
      Direct Array Index=$Particles.SetA
      Float Selection Array=DataInterface() {propertyValues {FloatData [0.2,0.26,0.34,0.13,0.28,0.34,0,0,0,0,0,0.45]}}
     ))) {isSetParameters=true}
   /SetVariables(Particles.CubeGlowScale:Vector3f=&Multiply_Vector(
     A=$Particles.CubeGlowScale
     B=&Multiply_Vector(
      A=&VectorFromFloat(Value=1.2)
      B=&SelectVectorFromArray(
       Direct Array Index=$Particles.SetA
       Vector Selection Array=DataInterface() {propertyValues {FloatData [<1 1 1>, <1 1 1>, <1 1 1>, <1 1 1>, <1 1 1>, <1 1 1>, <0.9 1 1>, <1 0.8 1>, <1.1 0.9 1>, <1 0.7 1>, <0.8 1 1>, <1 1 1>]}}
      )
     )
    )) {isSetParameters=true}
  ]
  particleUpdateScript [
   &ParticleState()
   &SkeletalMeshLocation(
    Bone / Socket Index=$Particles.SetB
    Bone Sampling Mode=::NewEnumerator4
    MaskSpawnGroup=false
    Orientation Sampling=::NewEnumerator1
    Skeletal Mesh=$System.Skeletal Mesh Reader
   )
   &SkeletalMeshLocation(
    Bone / Socket Index=$Particles.SetA
    Bone Sampling Mode=::NewEnumerator4
    MaskSpawnGroup=false
    Orientation Sampling=::NewEnumerator1
    Position Sampling=::NewEnumerator2
    Skeletal Mesh=$System.Skeletal Mesh Reader
   ) {module=SkeletalMeshLocation001}
   &SkeletalMeshLocation(
    Bone / Socket Index=$Particles.SetA
    Bone Sampling Mode=::NewEnumerator4
    MaskSpawnGroup=false
    Orientation Sampling=::NewEnumerator1
    Position Sampling=::NewEnumerator2
    Skeletal Mesh=$System.Skeletal Mesh Reader
   ) {module=SkeletalMeshLocation002}
   &self:ReadDeath(ChainTo:NiagaraPosition=$Particles.SkeletalMeshLocation001.SampledPosition) {module=ReadDeath}
   &ScaleMeshSize(Scale Factor=&MakeVector(
     X=&Divide_Float(
      A=&Lerp_Float(
       A=0
       B=$Particles.LightningLength
       Alpha=&FloatFromCurve(FloatCurve=DataInterface() {propertyValues {Curve {defaultValue=1, keys [{tangentMode=RCTM_User, time=0, value=0}, {tangentMode=RCTM_User, time=0.2, value=1}]}}})
      )
      B=100
     )
     Y=1
     Z=1
    ))
   &ScaleColor(
    Scale Alpha=&FloatFromCurve(FloatCurve=DataInterface() {propertyValues {Curve {
        defaultValue=1
        keys [
         {interpMode=RCIM_Cubic, tangentMode=RCTM_User, time=0, value=0}
         {interpMode=RCIM_Cubic, tangentMode=RCTM_User, time=0.1482, value=1}
         {tangentMode=RCTM_User, time=1, value=0}
        ]
       }}})
    Scale RGB=&VectorFromFloat(Value=&FN_FrontEndFloatModifier(Gameplay=5, Front End=5))
    ScaleA=true
    ScaleRGB=true
   )
   /SetVariables(
    Particles.CubeGlowColor:LinearColor=&MultiplyLinearColor(
     Linear Color=$Particles.CubeGlowColor
     Scale Factor=&MakeVector4FromVectorAndFloat(
      "Vector (RGB)"=&VectorFromFloat(Value=&FN_FrontEndFloatModifier(Gameplay=1.6, Front End=1.25))
      "Float (Alpha)"=&FloatFromCurve(
       FloatCurve=DataInterface() {propertyValues {Curve {
          defaultValue=1
          keys [
           {arriveTangent=-2.5249, interpMode=RCIM_Cubic, leaveTangent=-2.5249, tangentMode=RCTM_User, time=0, value=0.9847}
           {arriveTangent=-0.0405, interpMode=RCIM_Cubic, leaveTangent=-0.0405, tangentMode=RCTM_User, time=1, value=0}
          ]
         }}}
       Scale Curve=1.5
      )
     )
    )
    Particles.CubeGlowDyn:Vector4f=<1 1 1 1>
    Particles.GlowColor=&MultiplyLinearColor(
     Linear Color=$Particles.GlowColor
     Scale Factor=&MakeVector4FromVectorAndFloat(
      "Vector (RGB)"=<1.5 1.5 1.5>
      "Float (Alpha)"=&FloatFromCurve(
       FloatCurve=DataInterface() {propertyValues {Curve {
          defaultValue=1
          keys [
           {arriveTangent=9.0725, interpMode=RCIM_Cubic, leaveTangent=9.0725, tangentMode=RCTM_User, time=0, value=0}
           {arriveTangent=-2.4832, interpMode=RCIM_Cubic, leaveTangent=-2.4832, tangentMode=RCTM_User, time=0.0171, value=0.9847}
           {arriveTangent=-0.0405, interpMode=RCIM_Cubic, leaveTangent=-0.0405, tangentMode=RCTM_User, time=1, value=0}
          ]
         }}}
       Scale Curve=0
      )
     )
    )
    Particles.GlowDyn:Vector4f=<3 1 0 1>
   ) {isSetParameters=true}
   /SetVariables(Particles.Lifetime=&Divide_Float(
     A=$Particles.Initial.Lifetime
     B=&MergeHelper_FloatSwitch(0=2, 1=1.4, 2=?, 3=?, Index=$Particles.Initial.VisibilityTag, Num=::NewEnumerator0)
    )) {isSetParameters=true}
  ]
  rendererClasses [&NiagaraMeshRendererProperties, &NiagaraSpriteRendererProperties]
  renderers [
   {rendererClass=&NiagaraMeshRendererProperties, rendererIndex=0}
   {rendererClass=&NiagaraSpriteRendererProperties, rendererIndex=1}
   {rendererClass=&NiagaraMeshRendererProperties, rendererIndex=2}
  ]
 }
 SmallCubes {
  bEnabled=false
  emitterUpdateScript [
   &EmitterState(Life Cycle Mode=::NewEnumerator1, Loop Behavior=::NewEnumerator1, Loop Duration Mode=::NewEnumerator1, UseLoopDelay=false)
   &SpawnRate(SpawnRate=4, Use Spawn Probability=false)
  ]
  particleSpawnScript [
   &InitializeParticle(
    Color Mode=::NewEnumerator0
    Lifetime Max=1.5
    Lifetime Mode=::NewEnumerator1
    Mesh Scale Mode=::NewEnumerator1
    Mesh Uniform Scale Max=0.01
    Mesh Uniform Scale Min=0.008
    UsePositionOffset=false
    Write Material Random=true
   )
   &InitialMeshOrientation(Use Rotation Vector=false)
  ]
  particleUpdateScript [
   &ParticleState()
   &SkeletalMeshLocation(
    Bone / Socket Index=6
    Bone Sampling Mode=::NewEnumerator4
    MaskSpawnGroup=false
    Orientation Sampling=::NewEnumerator1
    Sampled Position Offset=&Lerp_Vector(
     A=&RandomRangeVector(Minimum=<3 3 0>, Maximum=<-3 -3 0>)
     B=&RandomRangeVector(Minimum=<5 5 10>, Maximum=<-5 -5 0>)
     Alpha=&FloatFromCurve(FloatCurve=DataInterface() {propertyValues {Curve {
         defaultValue=0
         keys [
          {arriveTangent=1.7914, interpMode=RCIM_Cubic, leaveTangent=1.7914, tangentMode=RCTM_User, time=0, value=0}
          {interpMode=RCIM_Cubic, tangentMode=RCTM_User, time=1, value=1}
         ]
        }}})
    )
    Sampled Position Offset Coordinate Space:ENiagaraSkelMeshTransforms=::NewEnumerator1
    Skeletal Mesh=$System.Skeletal Mesh Reader
   )
   &ScaleMeshSize(Scale Factor=&VectorFromFloat(Value=&FloatFromCurve(FloatCurve=DataInterface() {propertyValues {Curve {
         defaultValue=1
         keys [
          {arriveTangent=6.6152, interpMode=RCIM_Cubic, leaveTangent=6.6152, tangentMode=RCTM_User, time=0, value=0}
          {arriveTangent=3.1761, interpMode=RCIM_Cubic, leaveTangent=3.1761, time=0.1033, value=0.9024}
          {
           arriveTangent=0.0339
           arriveTangentWeight=0.5121
           interpMode=RCIM_Cubic
           leaveTangent=0.0339
           leaveTangentWeight=0.5121
           tangentMode=RCTM_User
           time=0.3149
           value=1
          }
          {arriveTangent=-3.0036, interpMode=RCIM_Cubic, leaveTangent=-3.0036, tangentMode=RCTM_User, time=1, value=0}
         ]
        }}})))
   &OffsetPosition(Position Offset=&MakeVector(
     X=0
     Y=0
     Z=&Waveform(
      Global Phase=&Add_Float(A=$Emitter.Age, B=&RandomRangeFloat(Minimum=0, Maximum=1))
      "Waveform [1]"=::NewEnumerator0
      "[1] Phase Scale"=&RandomRangeFloat(Minimum=0.3, Maximum=0.5)
      "[1] Frequency"=&RandomRangeFloat(Minimum=0.5, Maximum=1)
      "[1] Amplitude Min/Max"=<0 1>
     )
    ))
   &UpdateMeshOrientation(Rotation Rate=&RandomRangeFloat(Minimum=0.1, Maximum=0.3), Rotation Vector=<1 1 1>)
  ]
  rendererClasses [&NiagaraMeshRendererProperties]
  renderers [{rendererClass=&NiagaraMeshRendererProperties, rendererIndex=0}]
 }
}
summary {
 systemName=NS_Backpack_Season41_Reload_Cubes
 emitters [
  {
   emitterName=Cubes
   rendererClasses [
    &NiagaraMeshRendererProperties
    &NiagaraSpriteRendererProperties
    &NiagaraLightRendererProperties
   ]
  }
  {emitterName=Lightning, rendererClasses [&NiagaraMeshRendererProperties]}
  {bEnabled=false, emitterName=SmallCubes, rendererClasses [&NiagaraMeshRendererProperties]}
  {
   emitterName=LightningChains
   rendererClasses [&NiagaraMeshRendererProperties, &NiagaraSpriteRendererProperties]
  }
 ]
 userVariables [
  User.bIsFrontEnd {type=NiagaraBool}
  User.bIsFrontEndPreview {type=NiagaraBool}
  User.Variant {type=NiagaraFloat, defaultValue=0, description=""}
 ]
}
