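# Control Rig for the Season 41 backpack cosmetic that procedurally animates its cluster of cube bones from absolute time.
# A three-way sequence offsets seven small cubes with seeded sine/cosine drift and random-axis spin, then scale-pulses four center child cubes with ID-based phase shifts.
# Finally it spins and pulses MSTR_cube_center, propagating to its children.
# cube_center_02 has no pulse of its own, and two library functions are unused.
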ControlRigBlueprint /BRCosmetics/Accessories/FORT_Backpacks/Backpack_M_MED_Season_41/Meshes/M_MED_Season_41_Pack_CtrlRig
graphs [
 Add_10_ContainedGraph {nodes [
   n1 = RigVMFunctionEntryNode()
   RigVMFunctionReturnNode(Result=/Script/RigVM.RigVMFunction_MathVectorAdd(A=&RigVMFunction_MathVectorAdd(A=@n1.A, B=@n1.B) {template="Add::Execute(in A,in B,out Result)"}.Result, B=@n1.C) {is=struct RigVMUnitNode, template="Add::Execute(in A,in B,out Result)"}.Result)
  ]}
 Multiply_11_ContainedGraph {nodes [
   n1 = RigVMFunctionEntryNode()
   RigVMFunctionReturnNode(Result=&RigVMFunction_MathDoubleMul(B=@n1.C, A=&RigVMFunction_MathDoubleMul(A=@n1.A, B=@n1.B).Result).Result)
  ]}
 Multiply_11_ContainedGraph {nodes [
   n1 = RigVMFunctionEntryNode()
   RigVMFunctionReturnNode(Result=&RigVMFunction_MathDoubleMul(B=@n1.C, A=&RigVMFunction_MathDoubleMul(A=@n1.A, B=@n1.B).Result).Result)
  ]}
 Multiply_11_ContainedGraph {nodes [
   n1 = RigVMFunctionEntryNode()
   RigVMFunctionReturnNode(Result=&RigVMFunction_MathDoubleMul(B=@n1.C, A=&RigVMFunction_MathDoubleMul(A=@n1.A, B=@n1.B).Result).Result)
  ]}
 Multiply_12_ContainedGraph {nodes [
   n1 = RigVMFunctionEntryNode()
   RigVMFunctionReturnNode(Result=&RigVMFunction_MathDoubleMul(B=@n1.C, A=&RigVMFunction_MathDoubleMul(A=@n1.A, B=@n1.B).Result).Result)
  ]}
 Multiply_12_ContainedGraph {nodes [
   n1 = RigVMFunctionEntryNode()
   n2 = &RigVMFunction_MathDoubleMul(A=@n1.A, B=@n1.B)
   RigVMFunctionReturnNode(Result=&RigVMFunction_MathDoubleMul(B=@n1.D, A=&RigVMFunction_MathDoubleMul(B=@n1.C, A=@n2.Result).Result).Result)
  ]}
 Multiply_12_ContainedGraph {nodes [
   n1 = RigVMFunctionEntryNode()
   RigVMFunctionReturnNode(Result=&RigVMFunction_MathDoubleMul(B=@n1.C, A=&RigVMFunction_MathDoubleMul(A=@n1.A, B=@n1.B).Result).Result)
  ]}
 Multiply_18_ContainedGraph {nodes [
   n1 = RigVMFunctionEntryNode()
   RigVMFunctionReturnNode(Result=&RigVMFunction_MathDoubleMul(B=@n1.C, A=&RigVMFunction_MathDoubleMul(A=@n1.A, B=@n1.B).Result).Result)
  ]}
 Multiply_18_ContainedGraph {nodes [
   n1 = RigVMFunctionEntryNode()
   RigVMFunctionReturnNode(Result=&RigVMFunction_MathDoubleMul(B=@n1.C, A=&RigVMFunction_MathDoubleMul(A=@n1.A, B=@n1.B).Result).Result)
  ]}
 Multiply_18_ContainedGraph {nodes [
   n1 = RigVMFunctionEntryNode()
   RigVMFunctionReturnNode(Result=&RigVMFunction_MathDoubleMul(B=@n1.C, A=&RigVMFunction_MathDoubleMul(A=@n1.A, B=@n1.B).Result).Result)
  ]}
 New Function_ContainedGraph {nodes [
   n1 = RigVMFunctionEntryNode()
   n2 = RigVMDispatchNode(Struct=@n1.AmpTrs) {template="DISPATCH_RigVMDispatch_BreakStruct(in Struct,out Elements)"}
   n3 = /Script/RigVM.RigVMFunction_MathIntAdd(B=839, A=@n1.Seed) {is=struct RigVMUnitNode, template="Add::Execute(in A,in B,out Result)"}
   RigVMFunction_TimeLoop = /Script/RigVM.RigVMFunction_TimeLoop(Duration=10, Normalize=False, Speed=1) {is=struct RigVMUnitNode}
   Multiply_1 = RigVMAggregateNode(
    A=@n1.SpeedTrs
    C=/Script/RigVM.RigVMFunction_RandomFloat(
     Duration=-1
     Maximum=1
     Minimum=0
     Seed=/Script/RigVM.RigVMFunction_MathIntMul(B=168, A=@n3.Result) {is=struct RigVMUnitNode, template="Multiply::Execute(in A,in B,out Result)"}.Result
    ) {is=struct RigVMUnitNode, template="Random::Execute(in Seed,in Minimum,in Maximum,in Duration,out Result)"}.Result
    B=@RigVMFunction_TimeLoop.Absolute
   )
   n4 = &RigVMFunction_MathIntAdd(B=279, A=@n1.Seed) {template="Add::Execute(in A,in B,out Result)"}
   n5 = &RigVMFunction_MathDoubleAdd(
    A=@Multiply_1.Result
    B=&RigVMFunction_RandomFloat(
     Duration=-1
     Maximum=1000
     Minimum=0
     Seed=&RigVMFunction_MathIntMul(B=58, A=@n4.Result) {template="Multiply::Execute(in A,in B,out Result)"}.Result
    ) {template="Random::Execute(in Seed,in Minimum,in Maximum,in Duration,out Result)"}.Result
   )
   Sin = /Script/RigVM.RigVMFunction_MathDoubleSin(Value=@n5.Result) {is=struct RigVMUnitNode, template="Sin::Execute(in Value,out Result)"}
   Cos = /Script/RigVM.RigVMFunction_MathDoubleCos(Value=@n5.Result) {is=struct RigVMUnitNode, template="Cos::Execute(in Value,out Result)"}
   Multiply = &RigVMFunction_MathDoubleMul(B=2, A=@Cos.Result)
   Add = &RigVMFunction_MathIntAdd(B=232, A=@n1.Seed) {template="Add::Execute(in A,in B,out Result)"}
   n6 = &RigVMFunction_MathIntMul(B=54, A=@Add.Result) {template="Multiply::Execute(in A,in B,out Result)"}
   n7 = &RigVMFunction_MathIntMul(B=87, A=&RigVMFunction_MathIntAdd(B=443, A=@n1.Seed) {template="Add::Execute(in A,in B,out Result)"}.Result) {template="Multiply::Execute(in A,in B,out Result)"}
   n8 = &RigVMFunction_MathIntMul(B=23, A=&RigVMFunction_MathIntAdd(B=23, A=@n1.Seed) {template="Add::Execute(in A,in B,out Result)"}.Result) {template="Multiply::Execute(in A,in B,out Result)"}
   n9 = &RigVMFunction_MathIntAdd(B=839, A=@n1.Seed) {template="Add::Execute(in A,in B,out Result)"}
   Multiply_2 = RigVMAggregateNode(
    B=@n1.SpeedRot
    A=&RigVMFunction_RandomFloat(
     Duration=-1
     Maximum=1
     Minimum=0
     Seed=&RigVMFunction_MathIntMul(B=168, A=@n9.Result) {template="Multiply::Execute(in A,in B,out Result)"}.Result
    ) {template="Random::Execute(in Seed,in Minimum,in Maximum,in Duration,out Result)"}.Result
    C=@RigVMFunction_TimeLoop.Absolute
   )
   RigVMFunction_MathQuaternionFromAxisAndAngle = /Script/RigVM.RigVMFunction_MathQuaternionFromAxisAndAngle(
    Axis=RigVMDispatchNode(
     Elements=(X=0,Y=0,Z=0)
     Elements.X=&RigVMFunction_RandomFloat(Duration=-1, Maximum=1, Minimum=-1, Seed=@n6.Result) {template="Random::Execute(in Seed,in Minimum,in Maximum,in Duration,out Result)"}.Result
     Elements.Z=&RigVMFunction_RandomFloat(Duration=-1, Maximum=1, Minimum=-1, Seed=@n7.Result) {template="Random::Execute(in Seed,in Minimum,in Maximum,in Duration,out Result)"}.Result
     Elements.Y=&RigVMFunction_RandomFloat(Duration=-1, Maximum=1, Minimum=-1, Seed=@n8.Result) {template="Random::Execute(in Seed,in Minimum,in Maximum,in Duration,out Result)"}.Result
    ) {template="DISPATCH_RigVMDispatch_MakeStruct(in Elements,out Struct)"}.Struct
    Angle=@Multiply_2.Result
   ) {is=struct RigVMUnitNode, template="FromAxisAndAngle::Execute(in Axis,in Angle,out Result)"}
   > &RigUnit_OffsetTransformForItem(
    OffsetTransform=(Rotation=(X=0,Y=0,Z=0,W=1),Translation=(X=0,Y=0,Z=0),Scale3D=(X=1,Y=1,Z=1))
    Weight=1
    bPropagateToChildren=True
    Item=@n1.Bone
    OffsetTransform.Translation.X=&RigVMFunction_MathDoubleMul(B=@n2.Elements.X, A=@Sin.Result).Result
    OffsetTransform.Translation.Y=&RigVMFunction_MathDoubleMul(B=@n2.Elements.Y, A=@Multiply.Result).Result
    OffsetTransform.Translation.Z=&RigVMFunction_MathDoubleMul(B=@n2.Elements.Z, A=@Cos.Result).Result
    OffsetTransform.Rotation=@RigVMFunction_MathQuaternionFromAxisAndAngle.Result
   )
   > RigVMFunctionReturnNode()
  ]}
 New Function_ContainedGraph {nodes [
   n1 = RigVMFunctionEntryNode()
   n2 = &RigUnit_GetTransform(Item=(Type=Bone,Name="MSTR_cube_center"), bInitial=False)
   RigVMFunction_MathQuaternionToAxisAndAngle = /Script/RigVM.RigVMFunction_MathQuaternionToAxisAndAngle(Value=@n2.Transform.Rotation) {is=struct RigVMUnitNode, template="ToAxisAndAngle::Execute(in Value,out Axis,out Angle)"}
   RigVMFunction_TimeLoop = &RigVMFunction_TimeLoop(Duration=10, Normalize=False, Speed=1)
   Multiply_2 = RigVMAggregateNode(C=1, D=0.4, A=@n1.SpeedRot, B=@RigVMFunction_TimeLoop.Absolute)
   n3 = &RigVMFunction_MathDoubleSub(B=&RigVMFunction_MathDoubleMul(B=0, A=@n1.AmpTrs_1).Result, A=@Multiply_2.Result)
   RigVMFunction_MathQuaternionFromAxisAndAngle = &RigVMFunction_MathQuaternionFromAxisAndAngle(
    Axis=/Script/RigVM.RigVMFunction_MathTransformRotateVector(
     Transform=(Rotation=(X=0.5,Y=-0.5,Z=0.5,W=0.5),Translation=(X=0,Y=0,Z=0),Scale3D=(X=1,Y=1,Z=1))
     Vector=@RigVMFunction_MathQuaternionToAxisAndAngle.Axis
    ) {is=struct RigVMUnitNode, template="Rotate Vector::Execute(in Transform,in Vector,out Result)"}.Result
    Angle=@n3.Result
   ) {template="FromAxisAndAngle::Execute(in Axis,in Angle,out Result)"}
   > RigUnit_SetTransform = &RigUnit_SetTransform(
    Value=(Rotation=(X=0,Y=0,Z=0,W=1),Translation=(X=0,Y=0,Z=0),Scale3D=(X=1,Y=1,Z=1))
    Weight=1
    bInitial=False
    Value.Rotation=/Script/RigVM.RigVMFunction_MathQuaternionMul(
     A=&RigUnit_GetTransform(bInitial=False, Item=@n1.Bone).Transform.Rotation
     B=@RigVMFunction_MathQuaternionFromAxisAndAngle.Result
    ) {is=struct RigVMUnitNode, template="Multiply::Execute(in A,in B,out Result)"}.Result
    Value.Translation=&RigUnit_GetTransform(bInitial=False, Item=@n1.Bone).Transform.Translation
    Item=@n1.Bone
   )
   > RigVMFunctionReturnNode()
   RigUnit_GetTransform = &RigUnit_GetTransform(Item=(Type=Control,Name="pack_ctrl"), bInitial=False)
   Subtract = /Script/RigVM.RigVMFunction_MathVectorSub(A=(X=0,Y=0,Z=0), B=@RigUnit_GetTransform.Transform.Translation) {is=struct RigVMUnitNode, template="Subtract::Execute(in A,in B,out Result)"}
   n4 = RigVMDispatchNode(Struct=@Subtract.Result) {template="DISPATCH_RigVMDispatch_BreakStruct(in Struct,out Elements)"}
   &RigVMFunction_MathDoubleAdd(A=@n4.Elements.X, B=@RigVMFunction_TimeLoop.Absolute)
   Multiply_17 = RigVMAggregateNode(B=1, C=1, A=@RigVMFunction_TimeLoop.Absolute)
   RigVMFunction_MathVectorLength = /Script/RigVM.RigVMFunction_MathVectorLength(Value=@Subtract.Result) {is=struct RigVMUnitNode, template="Length::Execute(in Value,out Result)"}
   Multiply_12 = RigVMAggregateNode(
    C=0.3
    A=&RigVMFunction_MathDoubleCos(Value=@Multiply_17.Result) {template="Cos::Execute(in Value,out Result)"}.Result
    B=@RigVMFunction_MathVectorLength.Result
   )
   Multiply_15 = RigVMAggregateNode(
    C=0.3
    B=@RigVMFunction_MathVectorLength.Result
    A=&RigVMFunction_MathDoubleSin(Value=@Multiply_17.Result) {template="Sin::Execute(in Value,out Result)"}.Result
   )
   Add_8 = RigVMAggregateNode(
    B=(X=0,Y=0,Z=0)
    C=(X=0,Y=0,Z=0)
    A=RigVMDispatchNode(Elements=(X=0,Y=0,Z=0), Elements.X=@Multiply_12.Result, Elements.Y=@Multiply_15.Result) {template="DISPATCH_RigVMDispatch_MakeStruct(in Elements,out Struct)"}.Struct
   )
   &RigVMFunction_MathDoubleAdd(A=@n4.Elements.Y, B=@RigVMFunction_TimeLoop.Absolute)
   Add = &RigVMFunction_MathIntAdd(B=232, A=@n1.Seed) {template="Add::Execute(in A,in B,out Result)"}
   n5 = &RigVMFunction_MathIntAdd(B=443, A=@n1.Seed) {template="Add::Execute(in A,in B,out Result)"}
   n6 = &RigVMFunction_MathIntAdd(B=23, A=@n1.Seed) {template="Add::Execute(in A,in B,out Result)"}
   RigVMDispatchNode(
    Elements=(X=0,Y=0,Z=0)
    Elements.X=&RigVMFunction_RandomFloat(
     Duration=-1
     Maximum=1
     Minimum=-1
     Seed=&RigVMFunction_MathIntMul(B=54, A=@Add.Result) {template="Multiply::Execute(in A,in B,out Result)"}.Result
    ) {template="Random::Execute(in Seed,in Minimum,in Maximum,in Duration,out Result)"}.Result
    Elements.Z=&RigVMFunction_RandomFloat(
     Duration=-1
     Maximum=1
     Minimum=-1
     Seed=&RigVMFunction_MathIntMul(B=87, A=@n5.Result) {template="Multiply::Execute(in A,in B,out Result)"}.Result
    ) {template="Random::Execute(in Seed,in Minimum,in Maximum,in Duration,out Result)"}.Result
    Elements.Y=&RigVMFunction_RandomFloat(
     Duration=-1
     Maximum=1
     Minimum=-1
     Seed=&RigVMFunction_MathIntMul(B=23, A=@n6.Result) {template="Multiply::Execute(in A,in B,out Result)"}.Result
    ) {template="Random::Execute(in Seed,in Minimum,in Maximum,in Duration,out Result)"}.Result
   ) {template="DISPATCH_RigVMDispatch_MakeStruct(in Elements,out Struct)"}
   &RigVMFunction_MathDoubleMul(A=1, B=3)
   &RigVMFunction_MathDoubleMul(A=1, B=3)
   &RigVMFunction_MathDoubleMul(A=1, B=3)
   n7 = &RigVMFunction_MathIntAdd(B=839, A=@n1.Seed) {template="Add::Execute(in A,in B,out Result)"}
   Multiply_1 = RigVMAggregateNode(
    A=@n1.SpeedTrs
    C=&RigVMFunction_RandomFloat(
     Duration=-1
     Maximum=1
     Minimum=0
     Seed=&RigVMFunction_MathIntMul(B=168, A=@n7.Result) {template="Multiply::Execute(in A,in B,out Result)"}.Result
    ) {template="Random::Execute(in Seed,in Minimum,in Maximum,in Duration,out Result)"}.Result
    B=@RigVMFunction_TimeLoop.Absolute
   )
   n8 = &RigVMFunction_MathIntAdd(B=279, A=@n1.Seed) {template="Add::Execute(in A,in B,out Result)"}
   n9 = &RigVMFunction_MathDoubleAdd(
    A=@Multiply_1.Result
    B=&RigVMFunction_RandomFloat(
     Duration=-1
     Maximum=1000
     Minimum=0
     Seed=&RigVMFunction_MathIntMul(B=58, A=@n8.Result) {template="Multiply::Execute(in A,in B,out Result)"}.Result
    ) {template="Random::Execute(in Seed,in Minimum,in Maximum,in Duration,out Result)"}.Result
   )
   Sin = &RigVMFunction_MathDoubleSin(Value=@n9.Result) {template="Sin::Execute(in Value,out Result)"}
   &RigVMFunction_MathDoubleMul(B=@n4.Elements.X, A=@Sin.Result)
   Cos = &RigVMFunction_MathDoubleCos(Value=@n9.Result) {template="Cos::Execute(in Value,out Result)"}
   Multiply = &RigVMFunction_MathDoubleMul(B=2, A=@Cos.Result)
   &RigVMFunction_MathDoubleMul(B=@n4.Elements.Y, A=@Multiply.Result)
   &RigVMFunction_MathDoubleMul(B=@n4.Elements.Z, A=@Cos.Result)
   &RigUnit_OffsetTransformForItem(
    ExecutePin=()
    OffsetTransform=(Rotation=(X=0,Y=0,Z=0.8039,W=0.5948),Translation=(X=0,Y=0,Z=0),Scale3D=(X=1,Y=1,Z=1))
    Weight=1
    bPropagateToChildren=True
    OffsetTransform.Rotation=@RigVMFunction_MathQuaternionFromAxisAndAngle.Result
    Item=@n1.Bone
   )
   &RigVMFunction_RandomFloat(
    Duration=-1
    Maximum=1
    Minimum=0
    Seed=&RigVMFunction_MathIntMul(B=168, A=&RigVMFunction_MathIntAdd(B=839, A=@n1.Seed) {template="Add::Execute(in A,in B,out Result)"}.Result) {template="Multiply::Execute(in A,in B,out Result)"}.Result
   ) {template="Random::Execute(in Seed,in Minimum,in Maximum,in Duration,out Result)"}
   &RigUnit_GetTransform(Item=(Type=Control,Name="pack_ctrl"), bInitial=False)
   &RigVMFunction_MathQuaternionFromAxisAndAngle(Angle=57.369, Axis=(X=0,Y=1,Z=0)) {template="FromAxisAndAngle::Execute(in Axis,in Angle,out Result)"}
   &RigVMFunction_MathQuaternionFromAxisAndAngle(Angle=0, Axis=(X=0,Y=1,Z=0)) {template="FromAxisAndAngle::Execute(in Axis,in Angle,out Result)"}
   RigVMFunction_MathQuaternionFromEuler = &RigVMFunction_MathQuaternionFromEuler(Euler=(X=0,Y=0,Z=0), Euler.Z=@n3.Result)
   RigVMFunction_MathQuaternionToEuler = &RigVMFunction_MathQuaternionToEuler(RotationOrder=ZYX, Value=@n2.Transform.Rotation)
  ]}
 New Function_ContainedGraph {nodes [
   n1 = RigVMFunctionEntryNode()
   RigVMFunction_TimeLoop = &RigVMFunction_TimeLoop(Duration=10, Normalize=False, Speed=1)
   Remap = &RigVMFunction_MathDoubleRemap(
    SourceMaximum=1
    SourceMinimum=-1
    bClamp=false
    TargetMaximum=@n1.AmpScaleMax
    TargetMinimum=@n1.AmpScaleMin
    Value=&RigVMFunction_MathDoubleSin(Value=&RigVMFunction_MathDoubleMul(A=@n1.SpeedScale, B=@RigVMFunction_TimeLoop.Absolute).Result) {template="Sin::Execute(in Value,out Result)"}.Result
   )
   RigVMFunction_MathVectorFromDouble = /Script/RigVM.RigVMFunction_MathVectorFromDouble(Value=@Remap.Result) {is=struct RigVMUnitNode, template="Cast::Execute(in Value,out Result)"}
   > &RigUnit_OffsetTransformForItem(
    OffsetTransform=(Rotation=(X=0,Y=0,Z=0,W=1),Translation=(X=0,Y=0,Z=0),Scale3D=(X=1,Y=1,Z=1))
    Weight=1
    bPropagateToChildren=True
    Item=@n1.Bone
    OffsetTransform.Translation=@n1.OffsetTrs
    OffsetTransform.Rotation=&RigVMFunction_MathQuaternionFromEuler(Euler=(X=0,Y=0,Z=0), Euler.Z=&RigVMFunction_MathDoubleMul(A=@n1.SpeedRot, B=@RigVMFunction_TimeLoop.Absolute).Result).Result
    OffsetTransform.Scale3D=@RigVMFunction_MathVectorFromDouble.Result
   )
   > RigVMFunctionReturnNode()
   &RigUnit_GetTransform(Item=(Type=Control,Name="pack_ctrl"), bInitial=False)
  ]}
 New Function_ContainedGraph {nodes [
   n1 = RigVMFunctionEntryNode()
   n2 = &RigVMFunction_MathDoubleMul(B=3, A=@n1.ID)
   RigVMFunction_TimeLoop = &RigVMFunction_TimeLoop(Duration=10, Normalize=False, Speed=1)
   n3 = &RigVMFunction_MathDoubleMul(A=@n1.SpeedScale, B=@RigVMFunction_TimeLoop.Absolute)
   Remap = &RigVMFunction_MathDoubleRemap(
    SourceMaximum=1
    SourceMinimum=-1
    bClamp=false
    Value=&RigVMFunction_MathDoubleSin(Value=&RigVMFunction_MathDoubleAdd(A=@n2.Result, B=@n3.Result).Result) {template="Sin::Execute(in Value,out Result)"}.Result
    TargetMinimum=@n1.AmpScaleMin
    TargetMaximum=@n1.AmpScaleMax
   )
   RigVMFunction_MathVectorFromDouble = &RigVMFunction_MathVectorFromDouble(Value=@Remap.Result) {template="Cast::Execute(in Value,out Result)"}
   > &RigUnit_OffsetTransformForItem(
    OffsetTransform=(Rotation=(X=0,Y=0,Z=0,W=1),Translation=(X=0,Y=0,Z=0),Scale3D=(X=1,Y=1,Z=1))
    Weight=1
    bPropagateToChildren=True
    Item=@n1.Bone
    OffsetTransform.Scale3D=@RigVMFunction_MathVectorFromDouble.Result
   )
   > RigVMFunctionReturnNode()
   &RigUnit_GetTransform(Item=(Type=Control,Name="pack_ctrl"), bInitial=False)
  ]}
 New Function_ContainedGraph {nodes [
   n1 = RigVMFunctionEntryNode()
   n2 = RigVMDispatchNode(Struct=@n1.AmpTrs) {template="DISPATCH_RigVMDispatch_BreakStruct(in Struct,out Elements)"}
   n3 = &RigVMFunction_MathIntAdd(B=839, A=@n1.Seed) {template="Add::Execute(in A,in B,out Result)"}
   RigVMFunction_TimeLoop = &RigVMFunction_TimeLoop(Duration=10, Normalize=False, Speed=1)
   Multiply_1 = RigVMAggregateNode(
    A=@n1.SpeedTrs
    C=&RigVMFunction_RandomFloat(
     Duration=-1
     Maximum=1
     Minimum=0
     Seed=&RigVMFunction_MathIntMul(B=168, A=@n3.Result) {template="Multiply::Execute(in A,in B,out Result)"}.Result
    ) {template="Random::Execute(in Seed,in Minimum,in Maximum,in Duration,out Result)"}.Result
    B=@RigVMFunction_TimeLoop.Absolute
   )
   n4 = &RigVMFunction_MathIntAdd(B=279, A=@n1.Seed) {template="Add::Execute(in A,in B,out Result)"}
   n5 = &RigVMFunction_MathDoubleAdd(
    A=@Multiply_1.Result
    B=&RigVMFunction_RandomFloat(
     Duration=-1
     Maximum=1000
     Minimum=0
     Seed=&RigVMFunction_MathIntMul(B=58, A=@n4.Result) {template="Multiply::Execute(in A,in B,out Result)"}.Result
    ) {template="Random::Execute(in Seed,in Minimum,in Maximum,in Duration,out Result)"}.Result
   )
   Sin = &RigVMFunction_MathDoubleSin(Value=@n5.Result) {template="Sin::Execute(in Value,out Result)"}
   n6 = &RigVMFunction_MathDoubleMul(B=@n2.Elements.X, A=@Sin.Result)
   Cos = &RigVMFunction_MathDoubleCos(Value=@n5.Result) {template="Cos::Execute(in Value,out Result)"}
   Multiply = &RigVMFunction_MathDoubleMul(B=2, A=@Cos.Result)
   n7 = &RigVMFunction_MathDoubleMul(B=@n2.Elements.Y, A=@Multiply.Result)
   n8 = &RigVMFunction_MathDoubleMul(B=@n2.Elements.Z, A=@Cos.Result)
   Add = &RigVMFunction_MathIntAdd(B=232, A=@n1.Seed) {template="Add::Execute(in A,in B,out Result)"}
   n9 = &RigVMFunction_MathIntMul(B=54, A=@Add.Result) {template="Multiply::Execute(in A,in B,out Result)"}
   n10 = &RigVMFunction_MathIntMul(B=87, A=&RigVMFunction_MathIntAdd(B=443, A=@n1.Seed) {template="Add::Execute(in A,in B,out Result)"}.Result) {template="Multiply::Execute(in A,in B,out Result)"}
   n11 = &RigVMFunction_MathIntMul(B=23, A=&RigVMFunction_MathIntAdd(B=23, A=@n1.Seed) {template="Add::Execute(in A,in B,out Result)"}.Result) {template="Multiply::Execute(in A,in B,out Result)"}
   n12 = &RigVMFunction_MathIntAdd(B=839, A=@n1.Seed) {template="Add::Execute(in A,in B,out Result)"}
   Multiply_2 = RigVMAggregateNode(
    B=@n1.SpeedRot
    A=&RigVMFunction_RandomFloat(
     Duration=-1
     Maximum=1
     Minimum=0
     Seed=&RigVMFunction_MathIntMul(B=168, A=@n12.Result) {template="Multiply::Execute(in A,in B,out Result)"}.Result
    ) {template="Random::Execute(in Seed,in Minimum,in Maximum,in Duration,out Result)"}.Result
    C=@RigVMFunction_TimeLoop.Absolute
   )
   RigVMFunction_MathQuaternionFromAxisAndAngle = &RigVMFunction_MathQuaternionFromAxisAndAngle(
    Axis=RigVMDispatchNode(
     Elements=(X=0,Y=0,Z=0)
     Elements.X=&RigVMFunction_RandomFloat(Duration=-1, Maximum=1, Minimum=-1, Seed=@n9.Result) {template="Random::Execute(in Seed,in Minimum,in Maximum,in Duration,out Result)"}.Result
     Elements.Z=&RigVMFunction_RandomFloat(Duration=-1, Maximum=1, Minimum=-1, Seed=@n10.Result) {template="Random::Execute(in Seed,in Minimum,in Maximum,in Duration,out Result)"}.Result
     Elements.Y=&RigVMFunction_RandomFloat(Duration=-1, Maximum=1, Minimum=-1, Seed=@n11.Result) {template="Random::Execute(in Seed,in Minimum,in Maximum,in Duration,out Result)"}.Result
    ) {template="DISPATCH_RigVMDispatch_MakeStruct(in Elements,out Struct)"}.Struct
    Angle=@Multiply_2.Result
   ) {template="FromAxisAndAngle::Execute(in Axis,in Angle,out Result)"}
   > &RigUnit_OffsetTransformForItem(
    OffsetTransform=(Rotation=(X=0,Y=0,Z=0,W=1),Translation=(X=0,Y=0,Z=0),Scale3D=(X=1,Y=1,Z=1))
    Weight=1
    bPropagateToChildren=True
    OffsetTransform.Translation=&RigVMFunction_MathVectorAdd(
     A=RigVMDispatchNode(Elements=(X=0,Y=0,Z=0), Elements.X=@n6.Result, Elements.Y=@n7.Result, Elements.Z=@n8.Result) {template="DISPATCH_RigVMDispatch_MakeStruct(in Elements,out Struct)"}.Struct
     B=@n1.OffsetTrs
    ) {template="Add::Execute(in A,in B,out Result)"}.Result
    Item=@n1.Bone
    OffsetTransform.Rotation=@RigVMFunction_MathQuaternionFromAxisAndAngle.Result
   )
   > RigVMFunctionReturnNode()
  ]}
 RigVMFunctionLibrary {nodes [
   " SmallerCubes" = RigVMCollapseNode(AmpTrs=(X=0,Y=0,Z=0), Bone=(Type=None,Name="None"), ExecuteContext=(), OffsetTrs=(X=0,Y=0,Z=0))
   CentrBigCube = RigVMCollapseNode(Bone=(Type=None,Name="None"), ExecuteContext=(), OffsetTrs=(X=0,Y=0,Z=0))
   MiddleCubes = RigVMCollapseNode(Bone=(Type=None,Name="None"), ExecuteContext=())
   New Function = RigVMCollapseNode(AmpTrs=(X=0,Y=0,Z=0), AxisRot=(X=0,Y=0,Z=0), Bone=(Type=None,Name="None"), ExecuteContext=())
   New Function_1 = RigVMCollapseNode(AmpTrs=(X=0,Y=0,Z=0), AxisRot=(X=0,Y=0,Z=0), Bone=(Type=None,Name="None"), ExecuteContext=())
  ]}
 RigVMFunction_Sequence_1_ContainedGraph {nodes [
   n1 = RigVMFunctionEntryNode()
   RigVMFunction_Sequence = &RigVMFunction_Sequence(ExecuteContext=@n1.ExecuteContext)
   n2 = &RigVMFunction_Sequence(ExecuteContext=@RigVMFunction_Sequence.B)
   RigVMFunctionReturnNode(A=@RigVMFunction_Sequence.A, B=@n2.A, C=@n2.B)
  ]}
 RigVMModel {nodes [
   RigUnit_BeginExecution = &RigUnit_BeginExecution()
   RigVMFunction_Sequence = RigVMAggregateNode(ExecuteContext=@RigUnit_BeginExecution.ExecutePin)
   " SmallerCubes" = RigVMFunctionReferenceNode(
    AmpTrs=(X=0.3,Y=1,Z=1.5)
    Bone=(Type=Bone,Name="cube_01")
    OffsetTrs=(X=0,Y=0,Z=0)
    Seed=43
    SpeedRot=1
    SpeedTrs=1.1237
    ExecuteContext=@RigVMFunction_Sequence.A
   )
   > " SmallerCubes_1" = RigVMFunctionReferenceNode(AmpTrs=(X=0.3,Y=1,Z=1.5), Bone=(Type=Bone,Name="cube_02"), OffsetTrs=(X=0,Y=0,Z=0), Seed=11, SpeedRot=1, SpeedTrs=1)
   > " SmallerCubes_2" = RigVMFunctionReferenceNode(AmpTrs=(X=0.3,Y=1,Z=1.5), Bone=(Type=Bone,Name="cube_03"), OffsetTrs=(X=-5,Y=0,Z=0), Seed=324, SpeedRot=1, SpeedTrs=0.8)
   > " SmallerCubes_3" = RigVMFunctionReferenceNode(AmpTrs=(X=0.3,Y=1,Z=1.5), Bone=(Type=Bone,Name="cube_04"), OffsetTrs=(X=-4,Y=0,Z=0), Seed=4534, SpeedRot=1, SpeedTrs=1.5043)
   > " SmallerCubes_4" = RigVMFunctionReferenceNode(AmpTrs=(X=0.3,Y=1,Z=1.5), Bone=(Type=Bone,Name="cube_05"), OffsetTrs=(X=0,Y=0,Z=-2), Seed=45, SpeedRot=1, SpeedTrs=0.8)
   > " SmallerCubes_5" = RigVMFunctionReferenceNode(AmpTrs=(X=0.3,Y=1,Z=1.5), Bone=(Type=Bone,Name="cube_06"), OffsetTrs=(X=0,Y=0,Z=0), Seed=645, SpeedRot=1, SpeedTrs=1.2)
   > " SmallerCubes_6" = RigVMFunctionReferenceNode(AmpTrs=(X=0.3,Y=0.3,Z=1.5), Bone=(Type=Bone,Name="cube_07"), OffsetTrs=(X=0,Y=1,Z=3), Seed=34, SpeedRot=1, SpeedTrs=1.5)
   MiddleCubes_3 = RigVMFunctionReferenceNode(AmpScaleMax=1, AmpScaleMin=0.7, Bone=(Type=Bone,Name="cube_center_01"), ID=1, SpeedScale=0.8, ExecuteContext=@RigVMFunction_Sequence.B)
   > MiddleCubes = RigVMFunctionReferenceNode(AmpScaleMax=1, AmpScaleMin=0.7, Bone=(Type=Bone,Name="cube_center_03"), ID=2, SpeedScale=0.8)
   > MiddleCubes_1 = RigVMFunctionReferenceNode(AmpScaleMax=1, AmpScaleMin=0.7, Bone=(Type=Bone,Name="cube_center_04"), ID=3, SpeedScale=0.8)
   > MiddleCubes_2 = RigVMFunctionReferenceNode(AmpScaleMax=1, AmpScaleMin=0.7, Bone=(Type=Bone,Name="cube_center_05"), ID=4, SpeedScale=0.8)
   CentrBigCube = RigVMFunctionReferenceNode(
    AmpScaleMax=1
    AmpScaleMin=0.9
    Bone=(Type=Bone,Name="MSTR_cube_center")
    OffsetTrs=(X=0,Y=0.1,Z=0)
    SpeedRot=-5
    SpeedScale=0.5
    ExecuteContext=@RigVMFunction_Sequence.C
   )
  ]}
]
hierarchy [
 root {loc=<0 0 0>, rot=<0 0 0 1>}
 pack {parent=root, loc=<0 -0.7316 0>, rot=<0 0 0 1>}
 cube_01 {parent=pack, loc=<2.537 -15.8878 23.3488>, rot=<0.9732 -0.1445 0.1415 0.109>}
 cube_02 {parent=pack, loc=<2.103 -15.9399 18.0045>, rot=<0.9112 -0.2443 -0.3109 -0.1157>}
 cube_03 {parent=pack, loc=<-0.2807 -17.0695 15.768>, rot=<0.9851 -0.0776 -0.1191 0.0967>}
 cube_04 {parent=pack, loc=<-5.4198 -14.0717 16.9677>, rot=<0.9423 -0.1661 0.1091 -0.2694>}
 cube_05 {parent=pack, loc=<8.8369 -13.9909 11.5282>, rot=<0.9118 -0.2528 -0.3235 0.0016>}
 cube_06 {parent=pack, loc=<-9.0812 -10.3532 9.2741>, rot=<0.9344 -0.1845 0.2935 0.0823>}
 cube_07 {parent=pack, loc=<2.5742 -10.6064 8.7446>, rot=<0.9542 -0.1672 0.195 0.153>}
 pack_null {parent=pack, type=Null, loc=<10 0 0>, rot=<0 0 0 1>}
 MSTR_cube_center {parent=pack, loc=<0.1822 -14.9694 5.0143>, rot=<0.9698 -0.0225 0.1516 -0.1895>}
 cube_center_01 {parent=MSTR_cube_center, loc=<-2.6263 -1.232 -0.3824>, rot=<0.3784 0.1652 -0.1411 -0.8998>}
 cube_center_02 {parent=MSTR_cube_center, loc=<0.1524 0.8168 -0.2973>, rot=<0.3171 0.022 -0.0664 -0.9458>}
 cube_center_03 {parent=MSTR_cube_center, loc=<0.2046 2.6341 -0.395>, rot=<0.3034 0.0779 -0.14 -0.9393>}
 cube_center_04 {parent=MSTR_cube_center, loc=<-0.8839 4.8412 -0.3544>, rot=<0.2015 0.2032 -0.1376 -0.9482>}
 cube_center_05 {parent=MSTR_cube_center, loc=<3.6262 -0.5977 0.8472>, rot=<0.3245 0.1976 -0.1491 -0.9129>}
 packSocket {parent=pack, type=Null, loc=<0 0 0>, rot=<0 0 0 1>}
]
variables [NewVar {type="TArray<FVector>"}]
